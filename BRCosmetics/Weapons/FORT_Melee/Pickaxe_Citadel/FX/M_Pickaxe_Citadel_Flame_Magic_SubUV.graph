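# Unlit translucent Niagara sprite material for the Citadel pickaxe's magic flame effect, built on an 8x8 inky smoke SubUV texture.
# Three slightly UV-offset samples form RGB channels, partly desaturated for subtle color fringing, then tinted by particle color for emissive.
# A dynamic parameter blends power-sharpened and linear masks shaped by Refine; Add boosts only emissive.
# Opacity multiplies that mask by particle alpha, with depth fade skipped on forward and mobile.
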Material /BRCosmetics/Weapons/FORT_Melee/Pickaxe_Citadel/FX/M_Pickaxe_Citadel_Flame_Magic_SubUV
graph {
 properties {
  bUseTranslucencyVertexFog=false
  bUsedWithNiagaraSprites=true
  blendMode=BLEND_Translucent
  pixelDepthOffsetMode=PDOM_Legacy
  shadingModel=MSM_Unlit
  usedShadingModels=Unlit
 }
 nodes [
  n1 = ParticleColor()
  n2 = TextureCoordinate()
  n3 = Multiply() {constA=1.5, constB=0.001}
  n4 = Add(A=@n2, B=@n3)
  n5 = TextureSample(UVs=@n2) {samplerType=SAMPLERTYPE_LinearColor, texture=/Game/Effects/Fort_Effects/Textures/Smoke/T_Inky_Smoke_SubUV_8x8}
  n6 = Subtract(A=@n2, B=@n3)
  n7 = Desaturation(
   None=MaterialFunctionCall(
    X=TextureSample(UVs=@n4) {samplerType=SAMPLERTYPE_LinearColor, texture=&T_Inky_Smoke_SubUV_8x8}.R
    Y=@n5.R
    Z=TextureSample(UVs=@n6) {samplerType=SAMPLERTYPE_LinearColor, texture=&T_Inky_Smoke_SubUV_8x8}.R
   ) {materialFunction=&MakeFloat3}.Result
   Fraction=Constant() {r=0.8}
  )
  n8 = VectorParameter() {parameterName=Refine, defaultValue=<rgba 1 1 1 0.5>}
  n9 = LinearInterpolate(
   A=Multiply(A=Power(Base=@n5.R, Exp=@n8.R), B=@n8.G)
   B=Multiply(A=@n5.R, B=@n8.B)
   Alpha=DynamicParameter() {defaultValue=<rgba 0.5 1 1 1>}.Param1
  )
  n10 = Add(A=@n9, B=ScalarParameter() {parameterName=Add})
  MaterialFunctionCall(Emissive=Multiply(A=@n1.RGB, B=Multiply(A=@n7, B=@n10))) {materialFunction=&MF_FX_EmissiveAdjustments}
  n11 = Multiply(A=@n1.A, B=@n9)
  n12 = DepthFade(Opacity=@n11, FadeDistance=ScalarParameter() {defaultValue=20, parameterName=FadeDistance})
  ShadingPathSwitch(Default=@n12, Deferred=@n12, Forward=@n11, Mobile=@n11)
 ]
}
