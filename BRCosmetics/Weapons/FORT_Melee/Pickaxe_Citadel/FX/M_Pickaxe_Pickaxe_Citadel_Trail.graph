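# Translucent, unlit, two-sided Niagara ribbon material for the Citadel pickaxe trail, tinted by a gold curve-atlas gradient and particle color.
# Panning flame and inky smoke textures, warped by noise-driven UV distortion and offset by a dynamic parameter, build the glow and opacity.
# Three offset flame samples create a mostly desaturated RGB split.
# Parameterized edge masks, gradient and view-angle falloff shape opacity, with depth fade skipped on forward and mobile.
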Material /BRCosmetics/Weapons/FORT_Melee/Pickaxe_Citadel/FX/M_Pickaxe_Pickaxe_Citadel_Trail
graph {
 properties {
  bUsedWithNiagaraRibbons=true
  blendMode=BLEND_Translucent
  pixelDepthOffsetMode=PDOM_Legacy
  shadingModel=MSM_Unlit
  twoSided=true
  usedShadingModels=Unlit
 }
 nodes [
  n1 = Constant() {r=1}
  n2 = CurveAtlasRowParameter(CurveTime=MaterialFunctionCall() {materialFunction=&LinearGradient}.UGradient) {
   atlas=/Game/Athena/LinearColorCurves/VFX_CurveAtlas
   curve=/Game/Weapons/FORT_Melee/Pickaxe_Citadel/FX/Curve_Pickaxe_Citadel_Gold
   defaultValue=65
  }
  n3 = Add(A=Multiply(A=TextureCoordinate() {uTiling=0.5, vTiling=0.5}, B=NamedRerouteUsage()), B=NamedRerouteUsage())
  n4 = NamedRerouteUsage()
  n5 = Constant2Vector() {r=0.2}
  n6 = NamedRerouteUsage()
  n7 = Add(A=Panner(Coordinate=Add(A=@n3, B=@n4), Speed=Multiply(A=@n5, B=@n6)) {speedX=0.2}, B=Multiply(A=NamedRerouteUsage()) {constB=10})
  n8 = Add(A=@n7) {constB=0.005}
  n9 = TextureSample(UVs=@n7) {samplerType=SAMPLERTYPE_LinearGrayscale, texture=/BRCosmetics/Weapons/FORT_Melee/Pickaxe_Cube_Ninja_Male/FX/T_FluidFlameTile}
  n10 = Subtract(A=@n7) {constB=0.005}
  n11 = Desaturation(
   None=MaterialFunctionCall(
    X=TextureSample(UVs=@n8) {samplerType=SAMPLERTYPE_LinearGrayscale, texture=&T_FluidFlameTile}.R
    Y=@n9.G
    Z=TextureSample(UVs=@n10) {samplerType=SAMPLERTYPE_LinearGrayscale, texture=&T_FluidFlameTile}.B
   ) {materialFunction=&MakeFloat3}.Result
   Fraction=Constant() {r=0.85}
  )
  n12 = Power(Base=@n11) {constExponent=7}
  n13 = Multiply(A=TextureCoordinate() {uTiling=0.3}, B=NamedRerouteUsage())
  n14 = Multiply(A=Constant2Vector() {r=0.4}, B=NamedRerouteUsage())
  n15 = Add(
   A=MaterialFunctionCall(XY=Panner(Coordinate=@n13, Speed=@n14) {speedX=0.4}) {materialFunction=&Swizzle}.YX
   B=Multiply(A=NamedRerouteUsage()) {constB=10}
  ) {constB=0}
  n16 = Power(Base=Power(Base=TextureSample(UVs=@n15) {texture=&T_Inky_Smoke_Tile}.R))
  n17 = Add(A=Add(A=Multiply(A=@n12) {constB=50}, B=Multiply(A=@n11) {constB=2}), B=Multiply(A=@n16) {constB=10})
  n18 = NamedRerouteUsage()
  n19 = Multiply(
   A=Multiply(
    A=Multiply(A=@n18) {constB=10}
    B=MaterialFunctionCall() {materialFunction=&LinearGradient}.UGradient
   )
   B=Constant2Vector() {r=1}
  )
  n20 = TextureCoordinate()
  n21 = Power(Base=TextureSample(UVs=Add(A=@n19, B=@n20)) {samplerSource=SSM_Clamp_WorldGroupSettings, texture=/Game/Effects/Fort_Effects/Textures/Smoke/T_SmokeTrailCoord1_Mask}.R)
  n22 = LinearInterpolate(A=Multiply(A=@n17), B=Power(Base=Multiply(A=@n17) {constB=5}) {constExponent=1}, Alpha=Saturate(None=Multiply(A=@n21, B=@n21)))
  n23 = ParticleColor()
  MaterialFunctionCall(Emissive=Multiply(A=LinearInterpolate(A=@n1, B=@n2) {constAlpha=0.4}, B=Multiply(A=@n22, B=@n23.RGB))) {materialFunction=&MF_FX_EmissiveAdjustments}
  n24 = Multiply(A=TextureCoordinate() {uTiling=2, vTiling=2}, B=NamedRerouteUsage())
  n25 = NamedRerouteUsage()
  n26 = TextureSample(UVs=Panner(Coordinate=Add(A=@n24, B=@n25)) {speedX=0.1, speedY=0.03}) {samplerType=SAMPLERTYPE_LinearGrayscale, texture=/Engine/Functions/Engine_MaterialFunctions02/ExampleContent/Textures/LowResBlurredNoise}
  n27 = Multiply(A=@n26.R) {constB=0.1}
  n28 = Multiply(A=TextureCoordinate() {uTiling=0.5}, B=NamedRerouteUsage())
  n29 = TextureSample(UVs=Add(A=@n27, B=Panner(Coordinate=@n28) {speedX=0.16, speedY=-0.025})) {texture=&T_Inky_Smoke_Tile}
  NamedRerouteDeclaration(None=Multiply(A=Multiply(A=@n27), B=Multiply(A=@n29.R) {constB=0.2})) {name=UV_Distortion, nodeColor=<rgba 1 0.6328 0 1>}
  n30 = Add(A=@n9.G, B=@n16)
  n31 = Multiply(A=@n21) {constB=5}
  n32 = Add(A=TextureCoordinate() {vTiling=1.1}, B=Constant2Vector() {g=-0.05})
  n33 = Multiply(A=Multiply(A=NamedRerouteUsage()) {constB=5}, B=Constant2Vector() {g=1})
  n34 = Multiply(A=MaterialFunctionCall(Float2=Add(A=@n32, B=@n33)) {materialFunction=&BreakOutFloat2Components}.G) {constB=2}
  n35 = Multiply(A=Multiply(A=Multiply(A=@n30) {constB=5}, B=Multiply(A=@n31, B=@n34))) {constB=10}
  n36 = MaterialFunctionCall() {materialFunction=&LinearGradient}
  n37 = Multiply(A=Multiply(A=Saturate(None=@n35), B=NamedRerouteUsage()), B=Power(Base=OneMinus(None=@n36.UGradient)) {constExponent=3})
  n38 = Abs(None=DotProduct(A=CameraVectorWS(), B=PixelNormalWS()))
  n39 = DepthFade(Opacity=Multiply(A=@n37, B=Saturate(None=@n38))) {fadeDistanceDefault=50}
  n40 = Multiply(A=@n23.A, B=ShadingPathSwitch(Default=@n39, Deferred=@n39, Forward=@n37, Mobile=@n37))
  NamedRerouteDeclaration(None=Saturate(None=Divide(A=@n40) {constB=3})) {name=Opacity, nodeColor=<rgba 0.7578 0 1 1>}
  n41 = DynamicParameter() {paramNames [Offset U, Offset V, Param3, Param4]}
  NamedRerouteDeclaration(None=MaterialFunctionCall(X=@n41.Offset U, Y=@n41.Offset U) {materialFunction=&MakeFloat2}.Result) {name=Offset UV, nodeColor=<rgba 1 0.5625 0 1>}
  n42 = ScalarParameter() {parameterName=Top Edge Level}
  n43 = ScalarParameter() {defaultValue=0.2, parameterName=Top Edge Softness}
  n44 = NamedRerouteUsage()
  n45 = MaterialFunctionCall(Float2=Add(A=TextureCoordinate(), B=Multiply(A=@n44) {constB=5})) {materialFunction=&BreakOutFloat2Components}
  n46 = ScalarParameter() {defaultValue=0.1, parameterName=Bottom Edge Level}
  n47 = ScalarParameter() {defaultValue=0.4, parameterName=Bottom Edge Softness}
  n48 = Multiply(
   A=SmoothStep(Min=@n42, Max=Add(A=@n42, B=@n43), Value=Multiply(A=@n45.G))
   B=SmoothStep(Min=@n46, Max=Add(A=@n46, B=@n47), Value=OneMinus(None=@n45.G))
  )
  n49 = ScalarParameter() {parameterName=End Level}
  n50 = TextureCoordinate() {coordinateIndex=1}
  n51 = SmoothStep(
   Min=@n49
   Max=Add(A=@n49, B=ScalarParameter() {defaultValue=0.2, parameterName=End Level Softness})
   Value=OneMinus(None=MaterialFunctionCall(Float2=@n50) {materialFunction=&BreakOutFloat2Components}.R)
  )
  NamedRerouteDeclaration(None=Saturate(None=Multiply(A=@n48, B=@n51))) {name=mask, nodeColor=<rgba 0 0.4609 1 1>}
  n52 = VectorParameter() {parameterName=Pan Speed Multiplier, defaultValue=<rgba 2 0 0 1>}
  NamedRerouteDeclaration(None=MaterialFunctionCall(X=@n52.R, Y=@n52.R) {materialFunction=&MakeFloat2}.Result) {name=Pan Speed Multiplier, nodeColor=<rgba 1 0 0.6328 1>}
  NamedRerouteDeclaration(None=Multiply(A=ScalarParameter() {defaultValue=1, parameterName=U_Tile}, B=ScalarParameter() {defaultValue=1, parameterName=V_Tile})) {name=UVTiling, nodeColor=<rgba 1 0.2344 0 1>}
 ]
}
