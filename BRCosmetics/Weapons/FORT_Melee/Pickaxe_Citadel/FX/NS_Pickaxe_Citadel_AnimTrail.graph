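# Niagara anim trail system for the Citadel pickaxe, drawing warm-tinted ribbon trails from skeletal mesh bones or sockets.
# Three active ribbon emitters layer a main color trail, a faint blade trail and a short-lived soft trail, each fading via alpha curves.
# The main trail seeds random dynamic material parameters per spawn group and dims by time of day.
# AnimTrail_Color001, a near-duplicate of the main trail, is disabled.
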NiagaraSystem /BRCosmetics/Weapons/FORT_Melee/Pickaxe_Citadel/FX/NS_Pickaxe_Citadel_AnimTrail
dependencies {
 usedDataInterfaces [&NiagaraDataInterfaceCurve, &NiagaraDataInterfaceArrayColor, &NiagaraDataInterfaceColorCurve]
 usedDynamicInputs []
 usedModules [&SystemState]
}
emitters {
 AnimTrail_Color {
  emitterUpdateScript [
   &EmitterState()
   &FN_SpawnAnimTrail_v2()
  ]
  particleSpawnScript [
   /Game/Effects/Niagara/Modules/AnimTrail/FN_InitializeAnimTrail/FN_InitializeAnimTrailBoneIndecies(First Bone or Socket Index=1)
   /Game/Effects/Niagara/Modules/AnimTrail/FN_InitializeAnimTrail/FN_InitializeAnimTrail_v3/FN_InitializeAnimTrail(Color:LinearColor=<rgba 1 0.8958 0.75 1>, Lifetime=0.35)
   &FN_SetSpawnGroup(Number of Spawn Groups=1, Set Ribbon ID and Ribbon Link Order=true)
   /SetVariables() {isSetParameters=true, module=SetVariables_#1}
   &DynamicMaterialParameters(
    Index 0 Param 1=/Game/Effects/Niagara/DynamicInputs/Random/FN_SeededRandomRangeFloat(Minimum:NiagaraFloat=0, Maximum:NiagaraFloat=1, Random Seed:NiagaraInt32=$Particles.SpawnGroup)
    Index 0 Param 2=&FN_SeededRandomRangeFloat(Minimum:NiagaraFloat=0, Maximum:NiagaraFloat=1, Random Seed:NiagaraInt32=$Particles.SpawnGroup)
    Module.Param0WriteEnabled=true
    Module.Param1WriteEnabled=true
    Module.Param2WriteEnabled=true
    Module.Param3WriteEnabled=true
   )
  ]
  particleUpdateScript [
   /Game/Effects/Niagara/Modules/AnimTrail/FN_UpdateAnimTrail/FN_UpdateAnimTrail_v3/FN_UpdateAnimTrail(Locked Particle RibbonWidth:NiagaraFloat=$Particles.RibbonWidth)
   &ScaleColor(
    Scale Alpha=&FloatFromCurve(FloatCurve=DataInterface() {propertyValues {Curve {keys [
         {arriveTangent=-5.8828, interpMode=RCIM_Cubic, leaveTangent=-5.8828, tangentMode=RCTM_User, time=0.05, value=1}
         {arriveTangent=-1.3333, interpMode=RCIM_Cubic, leaveTangent=-1.3333, time=0.2, value=0.3}
         {arriveTangent=-0.1392, interpMode=RCIM_Cubic, leaveTangent=-0.1392, tangentMode=RCTM_User, time=0.65, value=0.2}
         <at 0.8 0 RCIM_Cubic>
        ]}}})
    ScaleA=true
    ScaleRGB=true
   ) {module=ScaleColor001}
   &ScaleColor(
    Color Value To Scale=$Particles.Color
    Scale Alpha=&Multiply_Float(A=1, B=&FNTimeOfDayScalar(Night Scale=0.6, Morning Scale=0.7, Evening Scale=0.7))
    Scale RGB=&Multiply_VectorByFloat(Vector=<1 1 1>, Float=1)
    ScaleA=true
    ScaleRGB=true
   )
  ]
  rendererClasses [&NiagaraRibbonRendererProperties]
  renderers [{rendererClass=&NiagaraRibbonRendererProperties}]
 }
 AnimTrail_Color001 {
  bEnabled=false
  emitterUpdateScript [
   &EmitterState()
   &FN_SpawnAnimTrail_v2()
  ]
  particleSpawnScript [
   &FN_InitializeAnimTrailBoneIndecies(First Bone or Socket Index=1)
   &FN_InitializeAnimTrail(Lifetime=0.35)
   &FN_SetSpawnGroup(Number of Spawn Groups=1, Set Ribbon ID and Ribbon Link Order=true)
   /SetVariables() {isSetParameters=true, module=SetVariables_#1}
   &DynamicMaterialParameters(
    Index 0 Param 1=&FN_SeededRandomRangeFloat(Minimum:NiagaraFloat=0, Maximum:NiagaraFloat=1, Random Seed:NiagaraInt32=$Particles.SpawnGroup)
    Index 0 Param 2=&FN_SeededRandomRangeFloat(Minimum:NiagaraFloat=0, Maximum:NiagaraFloat=1, Random Seed:NiagaraInt32=$Particles.SpawnGroup)
    Module.Param0WriteEnabled=true
    Module.Param1WriteEnabled=true
    Module.Param2WriteEnabled=true
    Module.Param3WriteEnabled=true
   )
   &Color(Color=&SelectLinearColorFromArray(
     Array Sampling Mode=::NewEnumerator1
     Direct Array Index=&MakeIntFromFloat(
      Float=&FN_SeededRandomRangeFloat(Minimum:NiagaraFloat=0, Maximum:NiagaraFloat=1, Random Seed:NiagaraInt32=$Particles.SpawnGroup)
      Conversion Mode=::NewEnumerator1
     )
     Color Selection Array=DataInterface() {propertyValues {ColorData [<rgba 1 0.006 0.8643 1>, <rgba 0 1 0.3576 1>]}}
     Randomness Mode=::NewEnumerator2
     Random Seed=&MakeIntFromFloat(
      Float=&FN_SeededRandomRangeFloat(Minimum:NiagaraFloat=0, Maximum:NiagaraFloat=1, Random Seed:NiagaraInt32=$Particles.SpawnGroup)
      Conversion Mode=::NewEnumerator1
     )
     Fixed Random Seed=true
    )) {enabled=false}
   &FN_RibbonMaterial(Color Scale U=500, Erossion Scale U=500, Opacity Scale U=500) {enabled=false}
  ]
  particleUpdateScript [
   &FN_UpdateAnimTrail(Locked Particle RibbonWidth:NiagaraFloat=$Particles.RibbonWidth)
   &Color(Color=&ColorFromCurve(ColorCurve=DataInterface() {propertyValues {
       AlphaCurve {keys [<at 0 1>,<at 1 1>]}
       BlueCurve {keys [<at 0.0604 1.8646>, {arriveTangent=-0.0001, leaveTangent=-0.0001, time=0.298, value=1}]}
       GreenCurve {keys [{arriveTangent=-1720.738, time=0.0604, value=1.4583}, {arriveTangent=-0.4891, leaveTangent=-0.4891, time=0.298, value=0.9375}]}
       RedCurve {keys [{arriveTangent=30.0741, time=0.0604, value=2}, {arriveTangent=52.7832, leaveTangent=52.7832, time=0.298, value=0.999}]}
      }})) {enabled=false, module=Color001}
   &ScaleColor(
    Scale Alpha=&FloatFromCurve(FloatCurve=DataInterface() {propertyValues {Curve {keys [
         {arriveTangent=-5.8828, interpMode=RCIM_Cubic, leaveTangent=-5.8828, tangentMode=RCTM_User, time=0.05, value=1}
         {arriveTangent=-1.3333, interpMode=RCIM_Cubic, leaveTangent=-1.3333, time=0.2, value=0.3}
         {arriveTangent=-0.1392, interpMode=RCIM_Cubic, leaveTangent=-0.1392, tangentMode=RCTM_User, time=0.65, value=0.2}
         <at 0.8 0 RCIM_Cubic>
        ]}}})
    ScaleA=true
    ScaleRGB=true
   ) {module=ScaleColor001}
   &ScaleColor(
    Color Value To Scale=$Particles.Color
    Scale Alpha=&Multiply_Float(A=1, B=&FNTimeOfDayScalar(Night Scale=0.3, Morning Scale=0.5, Evening Scale=0.5))
    Scale RGB=&Multiply_VectorByFloat(Vector=<1 1 1>, Float=1)
    ScaleA=true
    ScaleRGB=true
   )
  ]
  rendererClasses [&NiagaraRibbonRendererProperties]
  renderers [{rendererClass=&NiagaraRibbonRendererProperties}]
 }
 BladeTrail003 {
  emitterUpdateScript [
   &EmitterState()
   &FN_SpawnAnimTrail(AnimTrailWidthMode=::FromSecond)
  ]
  particleSpawnScript [
   /Game/Effects/Niagara/Modules/AnimTrail/FN_InitializeAnimTrailBoneIndecies_v2/FN_InitializeAnimTrailBoneIndecies(AnimTrail Selection=1)
   /Game/Effects/Niagara/Modules/AnimTrail/FN_InitializeAnimTrail/FN_InitializeAnimTrail_v5/FN_InitializeAnimTrail(
    AnimTrailWidth=8
    Color=&MultiplyLinearColor(Linear Color=<rgba 1.5 1.3438 1.125 0.1>, Scale Factor=<1 1 1 0.2>)
    Lifetime=0.2
   )
  ]
  particleUpdateScript [
   &SolveForcesAndVelocity(Clamp Velocity=false, Limit Acceleration=false)
   /Game/Effects/Niagara/Modules/AnimTrail/FN_UpdateAnimTrail/FN_UpdateAnimTrail_v5/FN_UpdateAnimTrail(Locked Particle RibbonWidth=1, SkeletalMesh=$User.SkeletalMesh) {module=FN_UpdateAnimTrail001}
   &ScaleColor(
    Scale Alpha=&FloatFromCurve(
     FloatCurve=DataInterface() {propertyValues {Curve {keys [
         {arriveTangent=-2.003, interpMode=RCIM_Cubic, leaveTangent=-2.003, tangentMode=RCTM_User, time=0.5, value=0.95}
         {interpMode=RCIM_Cubic, tangentMode=RCTM_User, time=1, value=0}
        ]}}}
     Scale Curve=0.5
    )
    ScaleA=true
    ScaleRGB=true
   )
  ]
  rendererClasses [&NiagaraRibbonRendererProperties]
  renderers [{rendererClass=&NiagaraRibbonRendererProperties}]
 }
 Trail_soft {
  emitterUpdateScript [
   &EmitterState(Life Cycle Mode=::NewEnumerator1, Loop Behavior=::NewEnumerator1, Loop Duration=0.2, UseLoopDelay=true)
   &FN_SpawnAnimTrail_v2()
  ]
  particleSpawnScript [
   /Game/Effects/Niagara/Modules/AnimTrail/FN_InitializeAnimTrail/FN_InitializeAnimTrailBoneIndecies(First Bone or Socket Index=1)
   /Game/Effects/Niagara/Modules/AnimTrail/FN_InitializeAnimTrail/FN_InitializeAnimTrail(AnimTrailWidth=1, Color=<rgba 0.6 0.4875 0.33 0.35>, Lifetime=0.2)
   &SystemLocation(MaskSpawnGroup=false)
  ]
  particleUpdateScript [
   &ScaleColor(
    Scale Alpha=&Multiply_Float(
     A=&FloatFromCurve(FloatCurve=DataInterface() {propertyValues {Curve {keys [
          {interpMode=RCIM_Cubic, leaveTangent=-2.3646, tangentMode=RCTM_Break, time=0, value=1}
          <at 0.6689 0.0233>
          {arriveTangent=-0.2139, interpMode=RCIM_Cubic, tangentMode=RCTM_Break, time=1, value=0}
         ]}}})
     B=&FNTimeOfDayScalar(Night Scale=1, Morning Scale=1, Evening Scale=1)
    )
    ScaleA=true
    ScaleRGB=true
   ) {module=ScaleColor002}
   &FN_UpdateAnimTrail_v2()
  ]
  rendererClasses [&NiagaraRibbonRendererProperties]
  renderers [{rendererClass=&NiagaraRibbonRendererProperties}]
 }
}
summary {
 systemName=NS_Pickaxe_Citadel_AnimTrail
 emitters [
  {emitterName=BladeTrail003, rendererClasses [&NiagaraRibbonRendererProperties]}
  {emitterName=AnimTrail_Color, rendererClasses [&NiagaraRibbonRendererProperties]}
  {bEnabled=false, emitterName=AnimTrail_Color001, rendererClasses [&NiagaraRibbonRendererProperties]}
  {emitterName=Trail_soft, rendererClasses [&NiagaraRibbonRendererProperties]}
 ]
 userVariables [
  User.RibbonIDOffset {type=NiagaraInt32}
  User.SkeletalMesh {
   type=NiagaraDataInterfaceSkeletalMesh
   defaultValue {
    struct=NiagaraExt_VariableValue_DataInterface
    value {dataInterface=&self:NiagaraDataInterfaceSkeletalMesh_0}
   }
  }
 ]
}
